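# WBP_QRCodeGenerator is a user widget that displays a string as a QR code in its single root image, Image_QRCode.
# SetEncodedString generates a texture from the given string and applies it to the image brush without matching its size.
# PreConstruct collapses visibility on Nintendo Switch or its preview.
# At design time, PreConstruct also encodes the Fortnite website URL as a preview.
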WidgetBlueprint /QRCodeGenerator/Widgets/WBP_QRCodeGenerator
graphs [
 EventGraph {nodes [
   n1 = Event PreConstruct()
   > n2 = Branch(Condition=IsNintendoSwitchOrNintendoSwitchPreview!() -> (ReturnValue:bool))
   > n3 = SetVisibility(InVisibility=Collapsed)
   Branch(execute=[@n3.then, @n2.else], Condition=@n1.IsDesignTime)
   > SetEncodedString!(execute, self, StringToEncode:String=https://www.fortnite.com/) -> (then)
  ]}
 SetEncodedString {nodes [
   n1 = SetEncodedString!() -> (then, StringToEncode:String) {is=title K2Node_FunctionEntry}
   > n2 = GenerateQRCodeTextureFromString!(execute, StringToEncode:String=@n1) -> (then, ReturnValue:Texture 2D Object Reference)
   > SetBrushFromTexture!(
    execute
    self:Image Object Reference=Get Image_QRCode() -> (Image_QRCode:Image Object Reference)
    Texture:Texture 2D Object Reference=@n2
    bMatchSize:bool=false
   ) -> (then)
  ]}
]
variables []
widgets {
 info {inheritedWidgetCount=0, parentClass=&UserWidget, rootWidgetClass=&Image, widgetCount=1}
 widgets [Image_QRCode {
   widgetClassPath=&Image
   bIsVariable=true
   widget=Image_QRCode
   properties {brush {
     bIsDynamicallyLoaded=false
     drawAs=Image
     imageType=NoImage
     mirroring=NoMirror
     resourceName=None
     resourceObject=null
     tiling=NoTile
     imageSize=<240 240>
     margin=<ltrb 0 0 0 0>
     outlineSettings {
      bUseBrushTransparency=false
      roundingType=HalfHeightRadius
      width=0
      color {colorUseRule=UseColor_Specified, specifiedColor=<rgba 0 0 0 0>}
      cornerRadii=<0 0 0 0>
     }
     tintColor {colorUseRule=UseColor_Specified, specifiedColor=<rgba 1 1 1 1>}
     uVRegion {bIsValid=false, max=<0 0>, min=<0 0>}
    }}
  }]
}
